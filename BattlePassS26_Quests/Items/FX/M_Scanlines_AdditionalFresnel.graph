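# Translucent unlit hologram material for meshes and particles, built on Atlas effect functions with scanlines, wiggle and a second fresnel multiplied into the first.
# Opacity combines fresnel, optional Z gradient, edge lines found by differencing offset normal-map samples, camera fade and scanlines, then optional masks and fades.
# Emissive blends base and glow colors, with optional inner lighting and intersection glow.
# Static switches add morph-target animation or axis rotation to vertices and normals.
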Material /BattlePassS26_Quests/Items/FX/M_Scanlines_AdditionalFresnel
graph {
 properties {
  bEnableResponsiveAA=true
  bTangentSpaceNormal=false
  bUsedWithMeshDeformer=true
  bUsedWithMeshParticles=true
  bUsedWithNiagaraMeshParticles=true
  bUsedWithParticleSprites=true
  bUsedWithSkeletalMesh=true
  blendMode=BLEND_Translucent
  numCustomizedUVs=4
  pixelDepthOffsetMode=PDOM_Legacy
  shadingModel=MSM_Unlit
  translucencyLightingMode=TLM_VolumetricPerVertexDirectional
  usedShadingModels=Unlit
 }
 nodes [
  ActorPositionWS()
  n1 = MaterialFunctionCall!(Distance Bias, Distance Divisor=ScalarParameter() {defaultValue=32, parameterName=Intersection Edge Divisor}, Scene Depth, Pixel Depth) {materialFunction=/Game/Effects/Fort_Effects/Materials/_Functions/MF_DistanceBehindTranslucency_01}
  n2 = Power(Base=@n1.Normalized Inverted, Exp=ScalarParameter() {defaultValue=6, parameterName=Intersection Edge Exponent})
  Add(A=@n2, B=Multiply(A=@n1.Normalized Inverted))
  Constant() {r=1}
  n3 = TextureSampleParameter2D() {group=Textures, parameterName=Alpha Texture, texture=&T_WhitesRGB}
  n4 = AppendVector(A=@n3.RGB, B=@n3.RGB)
  n5 = TextureSampleParameter2D() {group=Textures, parameterName=Base Color, texture=/Game/Effects/Fort_Effects/Effects/Atlas/Textures/T_Atlas_D}
  n6 = Multiply(A=@n5.R, B=ScalarParameter() {defaultValue=0.1, parameterName=AddOpacitytoVertexColorMask})
  n7 = Constant()
  n8 = VertexColor()
  n9 = Subtract(A=VectorParameter() {parameterName=ObjectPosition, defaultValue=<rgba 0 0 250 0>}.RGB, B=WorldPosition().XYZ)
  n10 = ConstantBiasScale(None=DotProduct(A=PixelNormalWS(), B=Normalize(VectorInput=@n9)))
  n11 = LinearInterpolate(
   A=ScalarParameter() {defaultValue=0.2, parameterName=Inner Opacity Dark}
   B=ScalarParameter() {defaultValue=1, parameterName=Inner Opacity Bright}
   Alpha=@n10
  )
  n12 = Constant() {r=1}
  n13 = ParticleColor()
  n14 = Constant() {r=1}
  n15 = StaticSwitchParameter(True=@n13.A, False=@n14) {defaultValue=true, parameterName=UsedInParticleSystems}
  n16 = StaticSwitchParameter(
   True=DynamicParameter() {defaultValue=<rgba 1 0 0 0>, paramNames [EdgeSpeed, BiasScale, Gradient Position, TriggerWiggle]}.BiasScale
   False=ScalarParameter() {defaultValue=0.8, parameterName=Opacity Bias Scale}
  ) {parameterName=Dynamic Parameter Bias Scale}
  n17 = TextureObjectParameter() {
   parameterName=Morph Target Normals
   samplerType=SAMPLERTYPE_LinearColor
   texture=/Game/Global/Textures/Default/Blanks/T_BlackVectorDisplacement
  }
  n18 = TextureObjectParameter() {
   parameterName=Morph Target Offsets
   samplerType=SAMPLERTYPE_LinearColor
   texture=&T_BlackVectorDisplacement
  }
  n19 = MaterialFunctionCall!(
   Morph Animation=Multiply(A=ScalarParameter() {defaultValue=0.35, parameterName=Time Speed}, B=Time() {period=1})
   Morph Normal=@n17
   Morph Texture=@n18
   Number of Morph Targets=ComponentMask(None=TextureProperty(Tex=@n17)) {g=true}
   0-1 Animation Value?=StaticBool() {value=true}
   Morph Target UV
  ) {materialFunction=/Engine/Functions/Engine_MaterialFunctions02/MAXScripts/MS_VertexAnimationTools_MorphTargets}
  n20 = Normalize(VectorInput=Transform(None=Constant3Vector() {constant=<rgba 0 0 1 0>}) {transformSourceType=TRANSFORMSOURCE_Local})
  n21 = Multiply(A=Time(), B=ScalarParameter() {defaultValue=0.1, parameterName=RotationSpeed})
  n22 = TextureSampleParameter2D() {
   group=Textures
   parameterName=Normal Map
   samplerType=SAMPLERTYPE_Normal
   texture=/Game/Effects/Fort_Effects/Effects/Atlas/Textures/T_Atlas_N
  }
  n23 = Transform(None=@n22.RGB)
  n24 = MaterialFunctionCall(Rotation Axis=@n20, Rotation Angle=@n21, World Space Vertex Normals=@n23) {materialFunction=&FixRotateAboutAxisNormals}
  n25 = StaticSwitchParameter(
   True=@n19."Pixel Shader World Vertex Normal (See tooltip)"
   False=StaticSwitchParameter(
    True=FeatureLevelSwitch(Default=@n24.New World Space Vertex Normal, ES2=@n23, ES3_1=@n23, SM4_REMOVED=@n24.New World Space Vertex Normal)
    False=@n23
   ) {parameterName=Use Rotation}
  ) {parameterName=Use Morph Targets?}
  n26 = MaterialFunctionCall(
   Brightness=ScalarParameter() {defaultValue=0.9, parameterName=Fresnel Brightness, sliderMin=4}
   Exponent=ScalarParameter() {defaultValue=10, parameterName=Fresnel Exponent}
   Vertex / Pixel Normal=@n25
  ) {materialFunction=&MF_Fresnel_FX}
  n27 = Multiply(
   A=LinearInterpolate(A=@n26.Fresnel, B=@n26.Fresnel, Alpha=ScalarParameter() {defaultValue=1, parameterName=Fresnel Inverse, sliderMax=1})
   B=NamedRerouteUsage()
  )
  n28 = Subtract(A=ComponentMask(None=WorldPosition().XYZ) {b=true}, B=ComponentMask(None=ObjectPositionWS()) {b=true})
  n29 = ScalarParameter() {defaultValue=0.5, group=ZGradient, parameterName=Gradient Thickness}
  n30 = OneMinus(None=Clamp(None=Divide(A=@n28, B=@n29)))
  n31 = ComponentMask(None=WorldPosition().XYZ) {b=true}
  n32 = ScalarParameter() {defaultValue=376, group=ZGradient, parameterName=StructureHeight}
  n33 = Add(
   A=ScalarParameter() {group=ZGradient, parameterName=ObjectPositionZ}
   B=Multiply(
    A=DynamicParameter() {defaultValue=<rgba 1 0 0 0>, paramNames [EdgeSpeed, BiasScale, Gradient Position, TriggerWiggle]}.Gradient Position
    B=@n32
   )
  )
  n34 = Clamp(None=Divide(A=Subtract(A=@n31, B=@n33), B=Multiply(A=@n32, B=@n29)) {constB=384})
  n35 = Multiply(A=Multiply(A=@n34, B=OneMinus(None=@n34)), B=ScalarParameter() {defaultValue=4, group=ZGradient, parameterName=Gradient Brightness}) {constB=4}
  n36 = Multiply(
   A=Add(A=Multiply(A=@n27, B=@n16), B=OneMinus(None=@n16))
   B=StaticSwitchParameter(True=StaticSwitchParameter(True=@n30, False=@n35) {parameterName=Use Z Gradient_LocalSpace}, False=Constant() {r=1}) {parameterName=Use Z Gradient?}
  )
  n37 = ScalarParameter() {defaultValue=0.0075, parameterName=Offset U}
  n38 = ScalarParameter() {defaultValue=0.0006, parameterName=Offset V}
  n39 = TextureSampleParameter2D(UVs=Add(A=AppendVector(A=@n37, B=@n38), B=TextureCoordinate())) {
   group=Textures
   parameterName=Normal Map
   samplerType=SAMPLERTYPE_Normal
   texture=&T_Atlas_N
  }
  n40 = TextureSampleParameter2D() {
   group=Textures
   parameterName=Normal Map
   samplerType=SAMPLERTYPE_Normal
   texture=&T_Atlas_N
  }
  n41 = DotProduct(A=Subtract(A=@n39.RGB, B=@n40.RGB), B=Constant() {r=1})
  n42 = StaticSwitchParameter(True=Constant(), False=Power(Base=Saturate(None=@n41), Exp=ScalarParameter() {defaultValue=8, parameterName=Blueprint Edge Exponent})) {parameterName=Use Morph Targets?}
  n43 = Multiply(A=@n42, B=ScalarParameter() {defaultValue=2.5, parameterName=Blueprint Edge Brightness})
  n44 = ScalarParameter() {defaultValue=200, parameterName=Near Camera Fade Distance}
  n45 = MaterialFunctionCall!(
   WPO Intensity A
   WPO Intensity B
   Time
   Bias=StaticSwitchParameter(
    True=DynamicParameter() {defaultValue=<rgba 1 0 0 0>, paramNames [EdgeSpeed, BiasScale, Gradient Position, TriggerWiggle]}.TriggerWiggle
    False=Constant()
   ) {parameterName=Use Param4 for Wiggle}
  ) {materialFunction=/Game/Effects/Fort_Effects/Effects/Atlas/Materials/MF_Atlas_HologramWiggle_01}
  n46 = MaterialFunctionCall!(
   Number of Scanlines=ScalarParameter() {defaultValue=150, parameterName=Scanline Amount}
   Scanline Speed
   "Scanlines Thickness (Bias)"=ScalarParameter() {defaultValue=-0.15, parameterName=ScanlineSize}
   Scanlines Dark Value=ScalarParameter() {defaultValue=0.5, parameterName=Scanlines Dark}
   Scanlines Light Value=ScalarParameter() {defaultValue=1, parameterName=Scanlines Light}
   Block Opacity=StaticSwitchParameter(True=@n45.Intense 0-1, False=Constant()) {defaultValue=true, parameterName=Use Blocks?}
   UVs
   Time
   Block Repeat
  ) {materialFunction=/Game/Effects/Fort_Effects/Effects/Atlas/Materials/MF_Atlas_Scanlines_01}
  n47 = Multiply(
   A=Multiply(
    A=Add(A=@n36, B=@n43)
    B=MaterialFunctionCall(Distance=@n44) {materialFunction=&Near_Camera_Fade}.Result
   )
   B=@n46.Result
  )
  n48 = Add(
   A=Multiply(A=StaticSwitchParameter(True=@n11, False=@n12) {parameterName=Use Inner Light?}, B=Multiply(A=@n15, B=@n47))
   B=Multiply(A=@n46.Secondary Lines, B=ScalarParameter() {defaultValue=1, parameterName=Secondary Line Opacity})
  )
  n49 = StaticSwitchParameter(True=Add(A=LinearInterpolate(A=@n6, B=@n7, Alpha=@n8.R), B=@n48), False=@n48) {parameterName=UseVertexColorMask}
  n50 = StaticSwitchParameter(True=Multiply(A=StaticComponentMaskParameter(None=@n4) {defaultG=true, parameterName=Alpha Texture Component Mask}, B=@n49), False=@n49) {parameterName=UseAlphaTextureMask}
  n51 = TextureCoordinate()
  n52 = Multiply(
   A=Power(
    Base=ComponentMask(None=@n51) {g=true}
    Exp=ScalarParameter() {defaultValue=2, group="1) TopDownGradient", parameterName=TopGradientEdgePower, sortPriority=1}
   )
   B=ScalarParameter() {defaultValue=1, group="1) TopDownGradient", parameterName=TopGradientEdgeStrength, sortPriority=2}
  )
  n53 = OneMinus(None=@n51)
  n54 = Multiply(
   A=Power(
    Base=ComponentMask(None=@n53) {g=true}
    Exp=ScalarParameter() {defaultValue=2, group="1) TopDownGradient", parameterName=BottomGradientEdgePower, sortPriority=3}
   )
   B=ScalarParameter() {defaultValue=1, group="1) TopDownGradient", parameterName=BottomGradientEdgeStrength, sortPriority=4}
  )
  n55 = LinearInterpolate(
   A=@n50
   B=Multiply(A=@n50, B=Multiply(A=@n52, B=@n54))
   Alpha=ScalarParameter() {group="1) TopDownGradient", parameterName=UseTopDownGradientEdge, sortPriority=0}
  )
  n56 = ScalarParameter() {parameterName=DepthFadeDistance}
  n57 = LinearInterpolate(A=@n55, B=Multiply(A=@n55, B=DepthFade(FadeDistance=@n56)), Alpha=ScalarParameter() {parameterName=UseDepthFade})
  LinearInterpolate(
   A=@n57
   B=Multiply(A=ParticleColor().A, B=@n57)
   Alpha=ScalarParameter() {group="1.1) UseParticleColorOpacity", parameterName=UseParticleColorOpacity, sortPriority=0}
  )
  n58 = VectorParameter() {parameterName=Intersection Edge Color, defaultValue=<rgba 0.25 1.5 20 1>}
  n59 = StaticSwitchParameter(
   True=LinearInterpolate(
    A=ScalarParameter() {defaultValue=0.05, parameterName=Inner Lighting Dark}
    B=ScalarParameter() {defaultValue=3, parameterName=Inner Lighting Bright}
    Alpha=@n10
   )
   False=@n12
  ) {parameterName=Use Inner Light?}
  n60 = StaticSwitchParameter(True=Constant(), False=@n5.RGB) {parameterName=Use Morph Targets?}
  n61 = Add(A=@n60, B=ScalarParameter() {defaultValue=0.2, parameterName=Base Color Bias})
  n62 = ScalarParameter() {defaultValue=0.5, parameterName=Secondary Line Brightness}
  n63 = LinearInterpolate(
   A=LinearInterpolate(A=@n60, B=VectorParameter() {parameterName=Glow Color Inner, defaultValue=<rgba 0.1 0.175 1 1>}.RGB, Alpha=Clamp(None=@n61))
   B=VectorParameter() {parameterName=Glow Color Outer, defaultValue=<rgba 0.55 2.125 10 1>}.RGB
   Alpha=Add(A=@n27, B=Multiply(A=@n62, B=@n46.Secondary Lines))
  )
  n64 = Multiply(
   A=LinearInterpolate(A=Multiply(A=@n59, B=@n63), B=VectorParameter() {parameterName=Glow Color Lines, defaultValue=<rgba 7 6.0667 1.4 1>}.RGB, Alpha=@n42)
   B=StaticSwitchParameter(True=@n13.RGB, False=@n14) {defaultValue=true, parameterName=UsedInParticleSystems}
  )
  n65 = StaticSwitchParameter(True=LinearInterpolate(A=@n5.RGB, B=@n64, Alpha=VertexColor().R), False=@n64) {parameterName=UseVertexColorMask}
  n66 = StaticSwitchParameter(True=Add(A=Multiply(A=@n2, B=@n58.RGB), B=@n65), False=@n65) {parameterName=Intersection}
  MaterialFunctionCall(Emissive=@n66) {materialFunction=&MF_FX_EmissiveAdjustments}
  NamedRerouteDeclaration(None=MaterialFunctionCall(
    Brightness=ScalarParameter() {defaultValue=2, parameterName=Secondary Fresnel Brightness}
    Exponent=ScalarParameter() {defaultValue=8, parameterName=Secondary Fresnel Exp}
    Vertex / Pixel Normal=@n25
   ) {materialFunction=&MF_Fresnel_FX}.Fresnel) {name=SecondaryFresnel, nodeColor=<rgba 0 1 0.7422 1>}
  Reroute(None=@n66)
  n67 = RotateAboutAxis(NormalizedRotationAxis=@n20, RotationAngle=@n21, PivotPoint=ObjectPositionWS(), Position=WorldPosition().XYZ)
  n68 = StaticSwitchParameter(True=Add(A=@n45.Result, B=@n19.World Position Offset), False=@n45.Result) {parameterName=Use Morph Targets?}
  StaticSwitchParameter(True=FeatureLevelSwitch(Default=@n67, ES2=@n68, ES3_1=@n68, SM4_REMOVED=@n67), False=@n68) {parameterName=Use Rotation}
  n69 = Constant()
  StaticSwitchParameter(True=@n19."Custom UV 2 (Needed)", False=@n69) {parameterName=Use Morph Targets?}
  StaticSwitchParameter(True=@n19."Custom UV 3 (Needed)", False=@n69) {parameterName=Use Morph Targets?}
  n70 = TextureSampleParameter2D() {
   group=Textures
   parameterName=Normal Map
   samplerType=SAMPLERTYPE_Normal
   texture=&T_Atlas_N
  }
  StaticSwitchParameter(True=@n19."Pixel Shader World Vertex Normal (See tooltip)", False=Transform(None=@n70.RGB)) {parameterName=Use Morph Targets?}
  Subtract(A=@n28, B=ObjectRadius())
 ]
}
